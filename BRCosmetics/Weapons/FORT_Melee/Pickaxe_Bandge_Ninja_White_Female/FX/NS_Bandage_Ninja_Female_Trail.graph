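# Niagara weapon trail system for the female white Bandage Ninja pickaxe, driven by the User.SkeletalMesh parameter.
# Three ribbon emitters use FN AnimTrail modules between bone or socket pairs, each short-lived with curve-tapered width and its own color; one also animates material parameters.
# A SheenMesh_start emitter bursts short-lived glints located on the skeletal mesh, rendering a mesh plus two sprite flares with curve-animated position, size and color.
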NiagaraSystem /BRCosmetics/Weapons/FORT_Melee/Pickaxe_Bandge_Ninja_White_Female/FX/NS_Bandage_Ninja_Female_Trail
dependencies {
 usedDataInterfaces [&NiagaraDataInterfaceCurve, &NiagaraDataInterfaceVectorCurve, &NiagaraDataInterfaceColorCurve]
 usedDynamicInputs []
 usedModules [&SystemState]
}
emitters {
 Empty004 {
  emitterUpdateScript [
   &EmitterState()
   &FN_SpawnAnimTrail()
  ]
  particleSpawnScript [
   &InitializeParticle(UsePositionOffset=false, Write Material Random=true) {enabled=false}
   &FN_InitializeAnimTrailBoneIndecies(First Bone or Socket Index=1, Second Bone or Socket Index=3)
   &FN_InitializeAnimTrail(Color=&MakeLinearColorFromVectorAndFloat("Vector (RGB)"=<1 1 1>, "Float (Alpha)"=0.2), Lifetime=0.25)
  ]
  particleUpdateScript [
   &ParticleState()
   &FN_UpdateAnimTrail(Kill Locked particle on system deactivate=false, SkeletalMesh=$User.SkeletalMesh)
   &ScaleRibbonWidth(Ribbon Width Scale=&FloatFromCurve(FloatCurve=DataInterface() {propertyValues {Curve {keys [{interpMode=RCIM_Cubic, tangentMode=RCTM_User, time=0, value=1}, {interpMode=RCIM_Cubic, tangentMode=RCTM_User, time=1, value=0}]}}}))
  ]
  rendererClasses [&NiagaraRibbonRendererProperties]
  renderers [{rendererClass=&NiagaraRibbonRendererProperties}]
 }
 Empty005 {
  emitterUpdateScript [
   &EmitterState(
    Inactive Response=::NewEnumerator1
    Life Cycle Mode=::NewEnumerator1
    Loop Delay=&RandomRangeFloat(Minimum=0, Maximum=0)
    Loop Duration=0.4
    Recalculate Duration Each Loop=true
    UseLoopDelay=true
   )
   &FN_SpawnAnimTrail()
  ]
  particleSpawnScript [
   &InitializeParticle(UsePositionOffset=false, Write Material Random=true) {enabled=false}
   &FN_InitializeAnimTrailBoneIndecies()
   &FN_InitializeAnimTrail(AnimTrailWidth=0.7, Color=<rgba 20 20 20 1>, Lifetime=0.3)
  ]
  particleUpdateScript [
   &ParticleState()
   &FN_UpdateAnimTrail(Kill Locked particle on system deactivate=false, SkeletalMesh=$User.SkeletalMesh)
   &DynamicMaterialParameters(
    Index 0 Param 1=2
    Index 0 Param 2=2
    Index 0 Param 3=1
    Index 0 Param 4=&FloatFromCurve(FloatCurve=DataInterface() {propertyValues {Curve {keys [<at 0 1>,<at 1 0>]}}})
    Module.Param0WriteEnabled=true
    Module.Param1WriteEnabled=true
    Module.Param2WriteEnabled=true
    Module.Param3WriteEnabled=true
   )
   &ScaleRibbonWidth(Ribbon Width Scale=&FloatFromCurve(FloatCurve=DataInterface() {propertyValues {Curve {keys [
         {arriveTangent=-1.2579, interpMode=RCIM_Cubic, time=0, value=1}
         {arriveTangent=-0.9664, interpMode=RCIM_Cubic, leaveTangent=-0.9664, tangentMode=RCTM_User, time=1, value=0}
        ]}}}))
  ]
  rendererClasses [&NiagaraRibbonRendererProperties]
  renderers [{rendererClass=&NiagaraRibbonRendererProperties}]
 }
 Empty007 {
  emitterUpdateScript [
   &EmitterState()
   &FN_SpawnAnimTrail()
  ]
  particleSpawnScript [
   &InitializeParticle(UsePositionOffset=false, Write Material Random=true) {enabled=false}
   &FN_InitializeAnimTrailBoneIndecies(First Bone or Socket Index=1, Second Bone or Socket Index=0)
   &FN_InitializeAnimTrail(AnimTrailWidth=1.1, Color=<rgba 25 32.1857 50 0.02>, Lifetime=0.25)
  ]
  particleUpdateScript [
   &ParticleState()
   &FN_UpdateAnimTrail(Kill Locked particle on system deactivate=false, SkeletalMesh=$User.SkeletalMesh)
   &ScaleRibbonWidth(Ribbon Width Scale=&FloatFromCurve(FloatCurve=DataInterface() {propertyValues {Curve {keys [
         {arriveTangent=-1.2579, interpMode=RCIM_Cubic, time=0, value=1}
         {arriveTangent=-0.9664, interpMode=RCIM_Cubic, leaveTangent=-0.9664, tangentMode=RCTM_User, time=1, value=0}
        ]}}}))
  ]
  rendererClasses [&NiagaraRibbonRendererProperties]
  renderers [{rendererClass=&NiagaraRibbonRendererProperties}]
 }
 SheenMesh_start {
  emitterUpdateScript [
   &EmitterState(
    Life Cycle Mode=::NewEnumerator1
    Loop Behavior=::NewEnumerator1
    Loop Delay=&RandomRangeFloat(Minimum=0, Maximum=0.25, Evaluation Type=::NewEnumerator1)
    Loop Duration=0.5
    UseLoopDelay=true
   )
   &SpawnBurst_Instantaneous(Spawn Count=1, Use Spawn Probability=false)
   &SpawnBurst_Instantaneous(Spawn Count=1, Spawn Time=0.1, Use Spawn Probability=false) {module=SpawnBurst_Instantaneous001}
   &SpawnRate(SpawnRate=20, Use Spawn Probability=false) {enabled=false}
  ]
  particleSpawnScript [
   &Color(Color=&MakeLinearColorFromVectorAndFloat(
     "Vector (RGB)"=<1 0 0>
     "Float (Alpha)"=&ClampFloat(Float=&RandomRangeFloat(Minimum=0.4, Maximum=0.6, Evaluation Type=::NewEnumerator1))
    )) {enabled=false}
   /SetVariables(
    Particles.ColorFlare:LinearColor=&MultiplyLinearColor(Linear Color=$System.GlintColor, Scale Factor=<20 20 20 20>)
    Particles.PositionFlare:NiagaraPosition=$Particles.Position
   ) {isSetParameters=true}
   /Niagara/Modules/Spawn/Initialization/InitializeParticle(
    Color=<rgba 16 20 16.6667 3>
    Lifetime=0.2
    Module.Write Color=true
    Module.Write Lifetime=true
    Module.Write Mass=true
    Module.Write Material Random=true
    Module.Write Position=false
    Module.Write Scale=true
    Module.Write SpriteRotation=false
    Module.Write SpriteSize=true
    Position=<0 -1 0>
    Sprite Size=<1 2>
   )
   /NiagaraFN/Sprite/FN_SpriteAlignmentAndFacing(SpriteAlignment:Vector3f=<0 0 1>, SpriteFacing:Vector3f=<1 0 0>)
  ]
  particleUpdateScript [
   /SetVariables(
    Particles.PositionFlare:Vector3f=&VectorFromCurve(VectorCurve=DataInterface() {propertyValues {
       XCurve {keys [
         <at 0.0009 0 RCIM_Cubic>
         {arriveTangent=15.0216, interpMode=RCIM_Cubic, leaveTangent=15.0216, time=0.4, value=2.5}
         {arriveTangent=11.1923, interpMode=RCIM_Cubic, leaveTangent=11.1929, tangentMode=RCTM_User, time=0.6, value=9}
        ]}
       YCurve {keys [<at 0 0>,<at 1 0>]}
       ZCurve {keys [
         <at 0 35 RCIM_Cubic>
         {arriveTangent=94.8541, interpMode=RCIM_Cubic, leaveTangent=94.8537, tangentMode=RCTM_User, time=0.35, value=55}
         {arriveTangent=34.8214, interpMode=RCIM_Cubic, leaveTangent=34.8228, tangentMode=RCTM_User, time=0.6, value=75}
        ]}
      }})
    Particles.PositionFlare2:Vector3f=&MakeVector(
     X=1
     Y=&FloatFromCurve(FloatCurve=DataInterface() {propertyValues {Curve {keys [
          {arriveTangent=107.135, interpMode=RCIM_Cubic, leaveTangent=107.1348, tangentMode=RCTM_User, time=0, value=15}
          {arriveTangent=33.5313, interpMode=RCIM_Cubic, leaveTangent=33.5312, tangentMode=RCTM_User, time=0.5, value=30}
         ]}}})
     Z=69
    )
   ) {isSetParameters=true}
   &ParticleState()
   &ScaleSpriteSize(Scale Factor=&MakeVector2D(
     X=&FloatFromCurve(
      FloatCurve=DataInterface() {propertyValues {Curve {keys [
          {arriveTangent=7.953, interpMode=RCIM_Cubic, leaveTangent=7.953, tangentMode=RCTM_User, time=-0.0008, value=0.5}
          {interpMode=RCIM_Cubic, leaveTangent=-0.7143, time=0.225, value=1.8998}
          {arriveTangent=-0.1355, interpMode=RCIM_Cubic, leaveTangent=-0.1355, tangentMode=RCTM_User, time=0.6, value=0.1497}
         ]}}}
      Scale Curve=10
     )
     Y=&FloatFromCurve(
      FloatCurve=DataInterface() {propertyValues {Curve {keys [
          {arriveTangent=7.953, interpMode=RCIM_Cubic, leaveTangent=7.953, tangentMode=RCTM_User, time=-0.0008, value=0.5}
          {interpMode=RCIM_Cubic, leaveTangent=-0.6804, time=0.225, value=1.8998}
          {arriveTangent=-0.4678, interpMode=RCIM_Cubic, leaveTangent=-0.4678, tangentMode=RCTM_User, time=0.6, value=0.1692}
         ]}}}
      Scale Curve=30
     )
    ))
   /SetVariables(Particles.SpriteSizeFlare2:Vector2f=&MakeVector2D(
     X=&FloatFromCurve(
      FloatCurve=DataInterface() {propertyValues {Curve {keys [{arriveTangent=-2.2776, interpMode=RCIM_Cubic, leaveTangent=-2.2776, tangentMode=RCTM_User, time=0, value=1}, <at 1 0 RCIM_Cubic>]}}}
      Scale Curve=3
     )
     Y=&FloatFromCurve(
      FloatCurve=DataInterface() {propertyValues {Curve {keys [<at 0 1 RCIM_Cubic>, {arriveTangent=-1.3497, interpMode=RCIM_Cubic, leaveTangent=-1.3497, tangentMode=RCTM_User, time=1, value=0.5}]}}}
      Scale Curve=20
     )
    )) {isSetParameters=true}
   &Color(Color=&MakeLinearColorFromVectorAndFloat("Vector (RGB)"=<0.4208 0.6361 1>, "Float (Alpha)"=&ClampFloat(Float=0.8))) {enabled=false, module=Color001}
   &DynamicMaterialParameters(
    Index 0 Param 1=&FloatFromCurve(FloatCurve=DataInterface() {propertyValues {Curve {keys [
         {arriveTangent=2.513, interpMode=RCIM_Cubic, time=0, value=1.8}
         {arriveTangent=-0.725, interpMode=RCIM_Cubic, leaveTangent=-0.725, tangentMode=RCTM_User, time=0.5, value=1.05}
         <at 1 0.2>
        ]}}})
    Index 0 Param 2=1
    Index 0 Param 3=1
    Index 0 Param 4=1
    Module.Param0WriteEnabled=true
    Module.Param1WriteEnabled=true
    Module.Param2WriteEnabled=true
    Module.Param3WriteEnabled=true
   )
   &CameraOffset(Camera Offset Amount=2)
   /SetVariables(Particles.CameraOffsetFlare=2) {isSetParameters=true}
   /SetVariables(Particles.ColorFlare:LinearColor=&ColorFromCurve(ColorCurve=DataInterface() {propertyValues {
       AlphaCurve {keys [<at 0.2801 1 RCIM_Cubic>,<at 0.5 0 RCIM_Cubic>]}
       BlueCurve {keys [<at 0 1 RCIM_Cubic>,<at 0.2942 0.9 RCIM_Cubic>]}
       GreenCurve {keys [<at 0 1 RCIM_Cubic>,<at 0.2942 1 RCIM_Cubic>]}
       RedCurve {keys [<at 0 1 RCIM_Cubic>,<at 0.2942 0.8 RCIM_Cubic>]}
      }})) {isSetParameters=true}
   &SolveForcesAndVelocity(Clamp Velocity=false, Limit Acceleration=false)
   &GenerateDeathEvent(Gap Correction Amount=1.5)
   &SkeletalMeshLocation(
    Bone / Socket Index=5
    Bone Sampling Mode=::NewEnumerator4
    MaskSpawnGroup=false
    Position Sampling=::NewEnumerator1
    Skeletal Mesh=$User.SkeletalMesh
   )
  ]
  rendererClasses [&NiagaraMeshRendererProperties, &NiagaraSpriteRendererProperties]
  renderers [
   {rendererClass=&NiagaraMeshRendererProperties, rendererIndex=0}
   {rendererClass=&NiagaraSpriteRendererProperties, rendererIndex=1}
   {rendererClass=&NiagaraSpriteRendererProperties, rendererIndex=2}
  ]
 }
}
summary {
 systemName=NS_Bandage_Ninja_Female_Trail
 emitters [
  {emitterName=Empty004, rendererClasses [&NiagaraRibbonRendererProperties]}
  {
   emitterName=SheenMesh_start
   rendererClasses [&NiagaraMeshRendererProperties, &NiagaraSpriteRendererProperties]
  }
  {emitterName=Empty007, rendererClasses [&NiagaraRibbonRendererProperties]}
  {emitterName=Empty005, rendererClasses [&NiagaraRibbonRendererProperties]}
 ]
 userVariables [
  User.PhysicalSurface {type=NiagaraInt32}
  User.RibbonIDOffset {type=NiagaraInt32}
  User.SkeletalMesh {
   type=NiagaraDataInterfaceSkeletalMesh
   defaultValue {
    struct=NiagaraExt_VariableValue_DataInterface
    value {dataInterface=&self:NiagaraDataInterfaceSkeletalMesh_1}
   }
  }
 ]
}
